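# Niagara system for the WindTyro companion emote, layering lightning breath, smoke, sparks and glow meshes on skeletal-mesh bones or sockets.
# User.LightningBreath drives time-of-day-tinted lightning meshes with particle lights, while User.Smoke drives rising smoke and intermittent sparks.
# Spark bursts are timed against fixed moments minus User.SpawnTime, with spawn groups varying position and velocity.
# VariantColor picks spark and spike-glow colors; GlobalGlow sets eye and spike glow alpha.
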NiagaraSystem /CosmeticCompanions/FX/Builtins/WindTyro/Emote/NS_Companion_Emote_WindTyro
dependencies {
 usedDataInterfaces [
  &NiagaraDataInterfaceCurve
  &NiagaraDataInterfaceColorCurve
  &NiagaraDataInterfaceArrayColor
  &NiagaraDataInterfaceSkeletalMesh
 ]
 usedDynamicInputs [
  &SimulationPosition
  &MakeFloatFromVector4
  &MaskVectorBySpawnGroup
 ]
 usedModules [
  &self:SystemSpawnScript.SystemScriptSource.SystemScriptGraph.NiagaraNodeAssignment_0.SetVariables_#1
  /Game/Effects/Niagara/Modules/Utility/DaySequenceTimeOfDayColor
  &SystemState
 ]
}
emitters {
 Lightning {
  emitterUpdateScript [
   &EmitterState()
   &SpawnRate(SpawnRate=&Lerp_Float(A=0, B=30, Alpha=$User.LightningBreath), Use Spawn Probability=false)
  ]
  particleSpawnScript [
   &FN_SetSpawnGroup(Number of Spawn Groups=2)
   /SetVariables(Particles.VisibilityTag=$Particles.SpawnGroup) {isSetParameters=true}
   &InitializeParticle(
    Color=$System.DaySequenceTimeOfDayColor.TImeOfDayColor
    Lifetime Max=0.35
    Lifetime Min=0.175
    Lifetime Mode=::NewEnumerator1
    Mesh Scale Max=<2.5 2.5 2.5>
    Mesh Scale Min=<2 2 1.4>
    Mesh Scale Mode=::NewEnumerator2
    Sprite Size Mode=::NewEnumerator3
    UsePositionOffset=false
    Write Material Random=true
   )
   /SetVariables(
    Particles.Random=&FloatFromCurve(
     FloatCurve=DataInterface() {propertyValues {Curve {
        defaultValue=1
        keys [
         {interpMode=RCIM_Cubic, tangentMode=RCTM_User, time=0, value=0}
         {arriveTangent=2.4041, interpMode=RCIM_Cubic, leaveTangent=2.4041, tangentMode=RCTM_User, time=1, value=1}
        ]
       }}}
     CurveIndex=$Particles.MaterialRandom
    )
    Particles.SizeDetermination:NiagaraBool=&SetBoolByFloatComparison(A=&RandomRangeFloat(Minimum=0, Maximum=1), B=0.7)
   ) {isSetParameters=true}
   &InitialMeshOrientation(
    Mesh Axis To Orient=<0 0 1>
    Mesh Orientation Mode=::NewEnumerator2
    Orientation Vector=<0 1 -0.3>
    Rotation=&RandomRangeVector(
     Minimum=&MakeVector(
      X=&MakeCustomFloatFromBool(
       Bool=$Particles.SizeDetermination
       True Float=&RandomRangeFloat(Minimum=-0.1, Maximum=0.1)
       False Float=0
      )
      Y=0
      Z=-1
     )
     Maximum=<0 0 1>
    )
    Use Rotation Vector=true
   )
   &CameraOffset(Camera Offset Amount=&RandomRangeFloat(Minimum=-5, Maximum=5))
   /SetVariables() {enabled=false, isSetParameters=true}
   &AddVelocity(Velocity=<10 0 0>)
  ]
  particleUpdateScript [
   &ParticleState()
   &ScaleMeshSize(Scale Factor=&VectorFromFloat(Value=$User.LightningBreath)) {enabled=false, module=ScaleMeshSize001}
   &SolveForcesAndVelocity(Clamp Velocity=false, Limit Acceleration=false)
   &ScaleMeshSize(Scale Factor=&Multiply_VectorByFloat(
     Vector=&MakeVector(X=1, Y=1, Z=&Lerp_Float(A=0.5, B=0.7, Alpha=$Particles.Random))
     Float=&MaskFloat(
      Mask Source=1
      Mask Parameter=&MakeIntFromFloat(Float=&MakeCustomFloatFromBool(Bool=$Particles.Initial.SizeDetermination, True Float=1, False Float=0))
      Mask=0
      Masked Float=0.3
      Passthrough Float=&RandomRangeFloat(Minimum=0.12, Maximum=0.17)
     )
    ))
   &DynamicMaterialParameters(
    Index 0 Param 1=&MakeFloatFromVector(Vector=$Particles.Scale, Channel=::NewEnumerator0)
    Index 0 Param 2=&RandomRangeFloat(Minimum=0, Maximum=10)
    Index 0 Param 3=&FloatFromCurve(
     FloatCurve=DataInterface() {propertyValues {Curve {keys [{interpMode=RCIM_Cubic, leaveTangent=8.1085, tangentMode=RCTM_Break, time=0.8568, value=1}, {tangentMode=RCTM_User, time=1, value=0}]}}}
     Scale Curve=0.0175
    )
    Index 0 Param 4=0.4
    Index 1 Param 1=&FloatFromCurve(FloatCurve=DataInterface() {propertyValues {Curve {keys [
         {interpMode=RCIM_Cubic, leaveTangent=1.1034, tangentMode=RCTM_Break, time=0, value=0}
         {arriveTangent=2.2399, tangentMode=RCTM_User, time=0.5, value=1}
        ]}}})
    Index 1 Param 2=&FloatFromCurve(FloatCurve=DataInterface() {propertyValues {Curve {keys [<at 0.5 0>, {arriveTangent=3.2053, tangentMode=RCTM_User, time=0.9, value=1}]}}})
    Index 1 Param 3=&FloatFromCurve(FloatCurve=DataInterface() {propertyValues {Curve {keys [<at 0 0 RCIM_Cubic>, {arriveTangent=3.4223, interpMode=RCIM_Cubic, leaveTangent=3.4223, tangentMode=RCTM_User, time=1, value=1}]}}})
    Index 1 Param 4=&FloatFromCurve(
     FloatCurve=DataInterface() {propertyValues {Curve {keys [
         {interpMode=RCIM_Cubic, leaveTangent=0.1169, tangentMode=RCTM_Break, time=0, value=0}
         {arriveTangent=2.5708, interpMode=RCIM_Cubic, leaveTangent=2.5708, tangentMode=RCTM_User, time=1, value=1}
        ]}}}
     Scale Curve=&RandomRangeFloat(Minimum=1, Maximum=40)
    )
    Index 2 Param 1=&FloatFromCurve(FloatCurve=DataInterface() {propertyValues {Curve {keys [
         {interpMode=RCIM_Cubic, leaveTangent=-0.0025, tangentMode=RCTM_Break, time=0, value=0.5}
         {arriveTangent=-0.0316, interpMode=RCIM_Cubic, tangentMode=RCTM_Break, time=1, value=-0.2}
        ]}}})
    Index 2 Param 2=&FloatFromCurve(
     FloatCurve=DataInterface() {propertyValues {Curve {keys [
         {interpMode=RCIM_Cubic, tangentMode=RCTM_User, time=0, value=0}
         {arriveTangent=2.0522, interpMode=RCIM_Cubic, leaveTangent=2.0522, tangentMode=RCTM_User, time=1, value=1}
        ]}}}
     Scale Curve=2
    )
    Module.Index0Float0Write=true
    Module.Index0Float1Write=true
    Module.Index0Float2Write=true
    Module.Index0Float3Write=true
    Module.Index1Float0Write=true
    Module.Index1Float1Write=true
    Module.Index1Float2Write=true
    Module.Index1Float3Write=true
    Module.Index2Float0Write=true
    Module.Index2Float1Write=true
    Module.Index2Float2Write=false
    Module.Index2Float3Write=false
    Module.Index3Float0Write=false
    Module.Index3Float1Write=false
    Module.Index3Float2Write=false
    Module.Index3Float3Write=true
    Write Parameter Index 1=::NewEnumerator1
    Write Parameter Index 2=::NewEnumerator1
    Write Parameter Index 3=::NewEnumerator1
   )
   &UpdateMeshOrientation(Rotation Rate=&RandomRangeFloat(Minimum=-1, Maximum=1), Rotation Vector=<0 0 1>) {enabled=false}
   &ScaleColor(
    Scale Alpha=&FloatFromCurve(FloatCurve=DataInterface() {propertyValues {Curve {keys [
         {tangentMode=RCTM_User, time=0, value=1}
         {arriveTangent=-0.0405, interpMode=RCIM_Cubic, leaveTangent=-0.0405, tangentMode=RCTM_User, time=0.7009, value=1}
         {arriveTangent=-8.0089, interpMode=RCIM_Cubic, leaveTangent=-8.0089, tangentMode=RCTM_User, time=1, value=0}
        ]}}})
    Scale RGB=&Lerp_Vector(
     A=&VectorFromFloat(Value=&FloatFromCurve(
       FloatCurve=DataInterface() {propertyValues {Curve {keys [<at 0 1>,<at 0.1 1>,<at 0.2 2>,<at 0.3 1>,<at 0.6 1>,<at 0.7 2>,<at 0.8 1>,<at 1 1>]}}}
       Scale Curve=40
      ))
     B=&VectorFromFloat(Value=&RandomRangeFloat(Minimum=0.3, Maximum=8, Evaluation Type=::NewEnumerator1))
     Alpha=&FloatFromCurve(FloatCurve=DataInterface() {propertyValues {Curve {keys [
          {interpMode=RCIM_Cubic, leaveTangent=2.5294, tangentMode=RCTM_Break, time=0, value=0}
          {arriveTangent=-0.2166, interpMode=RCIM_Cubic, tangentMode=RCTM_Break, time=0.4, value=1}
         ]}}})
    )
    ScaleA=true
    ScaleRGB=true
   )
   /SetVariables(Particles.AirGlowColor:LinearColor=&MultiplyLinearColor(Linear Color=$Particles.Color, Scale Factor=<1 1 1 0.04>)) {isSetParameters=true}
   &SkeletalMeshLocation(
    Bone Sampling Mode=::NewEnumerator4
    MaskSpawnGroup=false
    Sampled Position Offset=&MaskVector(Mask Source=1, Mask Parameter=$Particles.SpawnGroup, Mask=0, Masked Vector=<0 20 5>, Passthrough Vector=<0 0 0>)
    Skeletal Mesh=$System.Skeletal Mesh
   )
   &FN_Light_Attributes(
    Color=$Particles.Color
    Create Color Attribute=true
    Create Exponent Attribute=true
    Create Radius Attribute=true
    Enable Particle Light Enable=true
    Exponent=1.5
    Light Enabled:NiagaraBool=&RandomBool()
    Radius=&RandomRangeFloat(Minimum=25, Maximum=40)
    Scale Color=&Multiply_Float(A=10, B=&RandomRangeFloat(Minimum=0.5, Maximum=2, Evaluation Type=::NewEnumerator1))
   )
  ]
  rendererClasses [&NiagaraMeshRendererProperties, &NiagaraLightRendererProperties]
  renderers [
   {rendererClass=&NiagaraMeshRendererProperties, rendererIndex=0}
   {rendererClass=&NiagaraMeshRendererProperties, rendererIndex=1}
   {rendererClass=&NiagaraMeshRendererProperties, rendererIndex=2}
   {rendererClass=&NiagaraLightRendererProperties, rendererIndex=3}
  ]
 }
 Smoke {
  emitterUpdateScript [
   &EmitterState()
   &SpawnBurst_Instantaneous(Spawn Count=3, Use Spawn Probability=false, UseLoopCountLimit=false) {enabled=false}
   &SpawnRate(SpawnRate=&Lerp_Float(A=0, B=15, Alpha=$User.Smoke), Use Spawn Probability=false)
  ]
  particleSpawnScript [
   &InitializeParticle(
    Color=&RandomRangeLinearColor(Minimum=<rgba 0.1562 0.1562 0.1562 1.5>, Maximum=<rgba 0 0 0 1>)
    Lifetime=&UniformRangedFloat(Minimum=1.1, Maximum=1.5)
    Sprite Size Mode=::NewEnumerator1
    Uniform Sprite Size Max=14
    Uniform Sprite Size Min=12
    UsePositionOffset=false
    Write Material Random=true
   ) {module=InitializeParticle001}
   /SetVariables(Particles.RandomFloat=&UniformRangedFloat(Minimum=0, Maximum=1)) {isSetParameters=true}
   &CylinderLocation(
    Cylinder Height=0
    Cylinder Height Midpoint=1
    Cylinder Radius=5
    MaskSpawnGroup=false
    Offset Mode:ENiagara_OffsetMode=::NewEnumerator0
    Override Seed=false
    UseOwnerScale=true
   )
   &SkeletalMeshLocation(
    Bone Sampling Mode=::NewEnumerator4
    MaskSpawnGroup=false
    Sampled Position Offset=&RandomRangeVector(Minimum=<0 3 -3>, Maximum=<0 10 -8>)
    Skeletal Mesh=$System.Skeletal Mesh
   )
   &AddVelocity(
    Module.Use Falloff Distance=false
    Origin Offset=&RandomRangeVector(Minimum=<0 0 0>, Maximum=<0 -10 0>)
    Override Seed=false
    Velocity Mode=::NewEnumerator1
    Velocity Origin=$Particles.SkeletalMeshLocation.SampledPosition
    Velocity Speed=&RandomRangeFloat(Minimum=10, Maximum=100)
   ) {module=AddVelocity001}
  ]
  particleUpdateScript [
   &ParticleState()
   &AccelerationForce(Acceleration=&Multiply_Vector(
     A=&UniformRangedVector(Minimum=<0 0 250>, Maximum=<0 0 300>)
     B=&VectorFromFloat(Value=&FloatFromCurve(FloatCurve=DataInterface() {propertyValues {Curve {defaultValue=1, keys [{tangentMode=RCTM_User, time=0, value=0.6}, {tangentMode=RCTM_User, time=0.4, value=1}]}}}))
    ))
   &Drag(Drag=4, Use Linear Drag=true, Use Rotational Drag=false)
   &DynamicMaterialParameters(
    Index 0 Param 1=&Add_Float(
     A=$Particles.RandomFloat
     B=&FloatFromCurve(FloatCurve=DataInterface() {propertyValues {Curve {keys [
          {arriveTangent=2.4104, time=0, value=1}
          {arriveTangent=0.2537, interpMode=RCIM_Cubic, leaveTangent=0.2537, tangentMode=RCTM_User, time=1, value=2}
         ]}}})
    )
    Index 0 Param 4=1
    Index 2 Param 1=&FloatFromCurve(FloatCurve=DataInterface() {propertyValues {Curve {keys [{interpMode=RCIM_Cubic, leaveTangent=0.9717, tangentMode=RCTM_Break, time=0.25, value=0.8}, <at 1 1>]}}})
    Index 2 Param 2=&FloatFromCurve(
     FloatCurve=DataInterface() {propertyValues {Curve {keys [<at 0 0.5 RCIM_Cubic>,<at 1 0 RCIM_Cubic>]}}}
     Scale Curve=&RandomRangeFloat(Minimum=0.2, Maximum=0.5)
    )
    Index 2 Param 3=&FloatFromCurve(FloatCurve=DataInterface() {propertyValues {Curve {keys [
         {interpMode=RCIM_Cubic, tangentMode=RCTM_Break, time=0, value=1}
         {interpMode=RCIM_Cubic, leaveTangent=11.1233, tangentMode=RCTM_Break, time=1, value=3}
        ]}}})
    Index 2 Param 4=&FloatFromCurve(
     FloatCurve=DataInterface() {propertyValues {Curve {keys [
         {arriveTangent=-0.4552, interpMode=RCIM_Cubic, leaveTangent=-0.4552, tangentMode=RCTM_User, time=0, value=1}
         {interpMode=RCIM_Cubic, leaveTangent=-1.2956, tangentMode=RCTM_Break, time=1, value=0.75}
        ]}}}
     Scale Curve=1.4
    )
    Module.Index0Float0Write=true
    Module.Index0Float1Write=true
    Module.Index0Float2Write=true
    Module.Index0Float3Write=true
    Module.Index2Float0Write=true
    Module.Index2Float1Write=true
    Module.Index2Float2Write=true
    Module.Index2Float3Write=true
    Write Parameter Index 2=::NewEnumerator1
   ) {module=DynamicMaterialParameters001}
   &ScaleSpriteSize(Scale Factor=&Vector2DFromFloat(Value=&FloatFromCurve(
      FloatCurve=DataInterface() {propertyValues {Curve {keys [
          {arriveTangent=3.358, interpMode=RCIM_Cubic, leaveTangent=3.358, tangentMode=RCTM_User, time=0, value=0.1773}
          {arriveTangent=0.8227, interpMode=RCIM_Cubic, leaveTangent=0.8227, time=0.1179, value=0.489}
          {interpMode=RCIM_Cubic, leaveTangent=0.0321, time=1, value=1}
         ]}}}
      Scale Curve=2
     )))
   &ScaleColor(
    Scale Alpha=&FloatFromCurve(FloatCurve=DataInterface() {propertyValues {Curve {keys [
         {arriveTangent=9191.0225, interpMode=RCIM_Cubic, leaveTangent=4.1064, tangentMode=RCTM_User, time=0.0001, value=0.0123}
         {arriveTangentWeight=0.0484, interpMode=RCIM_Cubic, leaveTangentWeight=0.2906, tangentMode=RCTM_User, time=0.1, value=1}
         {arriveTangent=-1.1111, interpMode=RCIM_Cubic, leaveTangent=-1.1111, time=0.8, value=0.75}
         <at 1 0 RCIM_Cubic>
        ]}}})
    ScaleA=true
    ScaleRGB=true
   )
   &SpriteRotationRate(Rotation Rate=&Multiply_Float(
     A=&Lerp_Float(A=-150, B=150, Alpha=$Particles.MaterialRandom)
     B=&Lerp_Float(A=1, B=0, Alpha=&Power(a=$Particles.NormalizedAge, b=0.7))
    ))
   &CameraOffset(Camera Offset Amount=&RandomRangeFloat(Minimum=10, Maximum=30))
   &InheritSourceMovement(ApplyForce=false, ApplyPosition=true, ApplyPositionFalloff=false)
   &SolveForcesAndVelocity(Clamp Velocity=false, Limit Acceleration=false)
  ]
  rendererClasses [&NiagaraSpriteRendererProperties]
  renderers [{rendererClass=&NiagaraSpriteRendererProperties}]
 }
 Sparks001 {
  emitterUpdateScript [
   &EmitterState()
   &SpawnRate(
    Spawn Probability=&RandomRangeFloat(Minimum=0, Maximum=1)
    SpawnRate=&Lerp_Float(A=0, B=6, Alpha=$User.Smoke)
    Use Spawn Probability=true
   )
   &SpawnBurst_Instantaneous(
    Spawn Count=20
    Spawn Time=&Subtract_Float(A=5.25, B=$User.SpawnTime)
    Use Spawn Probability=false
    UseLoopCountLimit=false
   )
   &SpawnBurst_Instantaneous(
    Spawn Count=6
    Spawn Time=&Subtract_Float(A=6.25, B=$User.SpawnTime)
    Use Spawn Probability=false
    UseLoopCountLimit=false
   ) {module=SpawnBurst_Instantaneous002}
   &SpawnBurst_Instantaneous(
    Spawn Count=3
    Spawn Group=1
    Spawn Time=&Subtract_Float(A=6.45, B=$User.SpawnTime)
    Use Spawn Probability=false
    UseLoopCountLimit=false
   ) {module=SpawnBurst_Instantaneous001}
  ]
  particleSpawnScript [
   /SetVariables(Particles.SparkIntensity:NiagaraFloat=&FloatFromCurve(
     FloatCurve=DataInterface() {propertyValues {Curve {keys [
         {arriveTangent=0.2996, leaveTangentWeight=0.2, time=0, value=0}
         {
          arriveTangent=0.7746
          arriveTangentWeight=0.5371
          interpMode=RCIM_Cubic
          leaveTangent=0.7746
          leaveTangentWeight=0.3581
          tangentMode=RCTM_User
          time=0.5
          value=0.4
         }
         {arriveTangentWeight=0.1333, interpMode=RCIM_Cubic, tangentMode=RCTM_User, time=1, value=1}
        ]}}}
     CurveIndex=&RandomRangeFloat(Minimum=0, Maximum=1)
    )) {isSetParameters=true}
   &InitializeParticle(
    Color=&MergeHelper_ColorSwitch(
     0=&LerpLinearColors(StartColor=<rgba 4 3.1583 2.1806 1>, EndColor=<rgba 5 3 1.5 1>, LerpFactor=$Particles.SparkIntensity)
     1=&LerpLinearColors(StartColor=<rgba 2.0384 3.4577 4 1>, EndColor=<rgba 0.4875 3.2189 5 1>, LerpFactor=$Particles.SparkIntensity)
     2=&LerpLinearColors(StartColor=<rgba 4 3.808 2.256 1>, EndColor=<rgba 5 4.1191 1.7465 1>, LerpFactor=$Particles.SparkIntensity)
     3=?
     Index=$User.VariantColor
     Num=::NewEnumerator1
    )
    Lifetime=&Multiply_Float(
     A=&Lerp_Float(A=1, B=0.5, Alpha=$Particles.SparkIntensity)
     B=&MaskFloatBySpawnGroup(Masked Float=1.5, Passthrough Float=3, Spawn Group Mask=1)
    )
    Mass=&Lerp_Float(A=0.5, B=0.8, Alpha=$Particles.SparkIntensity)
    Mass Mode=::NewEnumerator1
    Sprite Size Mode=::NewEnumerator3
    Uniform Sprite Size=&Multiply_Float(A=0.15, B=&Lerp_Float(A=2.5, B=5, Alpha=$Particles.SparkIntensity))
    UsePositionOffset=false
    Write Material Random=true
   )
   &SkeletalMeshLocation(Bone / Socket Index=1, Bone Sampling Mode=::NewEnumerator4, MaskSpawnGroup=false, Skeletal Mesh=$System.Skeletal Mesh)
   &SkeletalMeshLocation(
    Bone / Socket Index=1
    Bone Sampling Mode=::NewEnumerator4
    MaskSpawnGroup=false
    Position Sampling=::NewEnumerator2
    Sampled Position Offset=<1 -10 0>
    Skeletal Mesh=$System.Skeletal Mesh
   ) {module=SkeletalMeshLocation001}
   &SkeletalMeshLocation(
    Bone / Socket Index=2
    Bone Sampling Mode=::NewEnumerator4
    MaskSpawnGroup=false
    Position Sampling=::NewEnumerator2
    Sampled Position Offset=<1 -10 0>
    Skeletal Mesh=$System.Skeletal Mesh
   ) {module=SkeletalMeshLocation002}
   &SkeletalMeshLocation(
    Bone / Socket Index=2
    Bone Sampling Mode=::NewEnumerator4
    MaskSpawnGroup=false
    Position Sampling=::NewEnumerator2
    Sampled Position Offset=<1 0 0>
    Skeletal Mesh=$System.Skeletal Mesh
   ) {module=SkeletalMeshLocation003}
   /SetVariables(
    Particles.OffsetLocation:NiagaraPosition=&MaskVector(
     Mask Source=1
     Mask Parameter=$Emitter.SpawnGroup
     Mask=0
     Masked Vector=Linked() {linkedVariable {name=Particles.SkeletalMeshLocation001.SampledPosition, type=NiagaraPosition}}
     Passthrough Vector=Linked() {linkedVariable {name=Particles.SkeletalMeshLocation003.SampledPosition, type=NiagaraPosition}}
    )
    Particles.Position=/Niagara/DynamicInputs/Helpers/MaskPosition(
     Mask Source:NiagaraInt32=1
     Mask Parameter:NiagaraInt32=$Emitter.SpawnGroup
     Mask:NiagaraInt32=0
     Masked Position:NiagaraPosition=$Particles.Position
     Passthrough Position:NiagaraPosition=$Particles.SkeletalMeshLocation002.SampledPosition
    )
   ) {isSetParameters=true}
   &SphereLocation(
    MaskSpawnGroup=false
    Override Seed=false
    Sphere Radius=2
    Surface Only=true
    Surface Only Band Thickness=0.5
    UseNonUniformScale=false
    UseOffset=false
    UseSphereOrientation=false
   )
   &AddVelocity(
    Module.Use Falloff Distance=false
    Override Seed=false
    Velocity Mode=::NewEnumerator1
    Velocity Origin=$Particles.OffsetLocation
    Velocity Speed=&Multiply_Float(
     A=&MaskFloat(Mask Source=1, Mask Parameter=$Emitter.SpawnGroup, Mask=0, Masked Float=0.2, Passthrough Float=0.5)
     B=&Multiply_Float(
      A=&Multiply_Float(
       A=&RandomRangeFloat(Minimum=2000, Maximum=5000)
       B=&Lerp_Float(A=0.4, B=1, Alpha=$Particles.SparkIntensity)
      )
      B=&MaskFloatBySpawnGroup(Masked Float=0.25, Passthrough Float=1, Spawn Group Mask=1)
     )
    )
   )
   /SetVariables(Particles.Velocity=&Multiply_VectorByFloat(Vector=$Particles.Velocity, Float=0.4)) {isSetParameters=true}
  ]
  particleUpdateScript [
   &ParticleState()
   &DynamicMaterialParameters(
    Index 0 Param 3=&FloatFromCurve(
     FloatCurve=DataInterface() {propertyValues {Curve {keys [{arriveTangent=-7.7784, interpMode=RCIM_Cubic, time=0, value=3}, <at 0.4 1>]}}}
     CurveIndex=$Emitter.Age
    )
    Module.Index0Float0Write=false
    Module.Index0Float1Write=false
    Module.Index0Float2Write=true
    Module.Index0Float3Write=false
   )
   &VectorNoiseForce(
    Force Amount=&FloatFromCurve(
     FloatCurve=DataInterface() {propertyValues {Curve {
        defaultValue=1
        keys [
         {interpMode=RCIM_Cubic, tangentMode=RCTM_User, time=0.1, value=0}
         {interpMode=RCIM_Cubic, tangentMode=RCTM_User, time=0.6, value=1}
         {interpMode=RCIM_Cubic, leaveTangent=2.5, time=1, value=0.25}
        ]
       }}}
     Scale Curve=15000
    )
    Module.Use Random Noise Time Interval=true
    Random Noise Time Interval=0.5
   ) {enabled=false}
   &GravityForce(Gravity=&Multiply_VectorByFloat(Vector=<0 0 -980>, Float=1))
   &Drag(
    Drag=&Multiply_Float(
     A=&FloatFromCurve(
      FloatCurve=DataInterface() {propertyValues {Curve {
         defaultValue=1
         keys [
          {arriveTangent=1.0867, interpMode=RCIM_Cubic, leaveTangent=1.0867, tangentMode=RCTM_User, time=0, value=0}
          {tangentMode=RCTM_User, time=1, value=1}
         ]
        }}}
      Scale Curve=&RandomRangeFloat(Minimum=1, Maximum=3)
     )
     B=1
    )
    Use Linear Drag=true
    Use Rotational Drag=false
   )
   &SolveForcesAndVelocity(Clamp Velocity=false, Limit Acceleration=false)
   &ScaleColor(
    Linear Color Curve=DataInterface() {propertyValues {
      AlphaCurve {keys [<at 0 1>,<at 1 1>]}
      BlueCurve {keys [{arriveTangent=-2.9361, interpMode=RCIM_Cubic, leaveTangent=-2.9361, tangentMode=RCTM_User, time=0, value=1}, <at 1 0.4206>]}
      GreenCurve {keys [{arriveTangent=-1.4408, interpMode=RCIM_Cubic, leaveTangent=-1.4408, tangentMode=RCTM_User, time=0.1467, value=1}, <at 1 0.5003>]}
      RedCurve {keys [{arriveTangent=-0.8332, interpMode=RCIM_Cubic, leaveTangent=-0.8332, tangentMode=RCTM_User, time=0.1467, value=1}, <at 1 0.7691>]}
     }}
    Scale Mode=::NewEnumerator2
   )
   &ScaleSpriteSizeBySpeed(
    Max Scale Factor=&MakeVector2D(
     X=1
     Y=&Lerp_Float(
      A=&RandomRangeFloat(Minimum=15, Maximum=25)
      B=1
      Alpha=&FloatFromCurve(FloatCurve=DataInterface() {propertyValues {Curve {
          defaultValue=0
          keys [
           {arriveTangent=4.1785, interpMode=RCIM_Cubic, leaveTangent=4.1785, tangentMode=RCTM_User, time=0, value=0}
           {arriveTangent=1.1179, interpMode=RCIM_Cubic, leaveTangent=1.1179, tangentMode=RCTM_User, time=0.2042, value=0.6228}
           {arriveTangent=0.0224, interpMode=RCIM_Cubic, leaveTangent=0.0224, tangentMode=RCTM_User, time=1, value=1}
          ]
         }}})
     )
    )
    Velocity Threshold:NiagaraFloat=&Multiply_Float(A=5000, B=$User.VelocityScale)
   )
   &ScaleSpriteSize(
    Scale Factor=&Vector2DFromFloat(Value=&FloatFromCurve(
      FloatCurve=DataInterface() {propertyValues {Curve {keys [<at 0.5 1 RCIM_Cubic>, {arriveTangent=-3.8662, interpMode=RCIM_Cubic, leaveTangent=-3.8662, tangentMode=RCTM_User, time=1, value=0}]}}}
      Scale Curve=&Lerp_Float(A=1, B=12, Alpha=$System.DistanceScaleLerp)
     ))
    Scale Sprite Size Mode=::NewEnumerator2
   )
  ]
  rendererClasses [&NiagaraSpriteRendererProperties]
  renderers [{rendererClass=&NiagaraSpriteRendererProperties}]
 }
 YellowLightMesh {
  emitterSpawnScript [/SetVariables() {isSetParameters=true}]
  emitterUpdateScript [
   &EmitterState()
   &SpawnBurst_Instantaneous(Use Spawn Probability=false, UseLoopCountLimit=false)
  ]
  particleSpawnScript [
   &PartitionParticles()
   &InitializeParticle(
    Color=<rgba 50 29.8526 5.3121 1>
    Lifetime=500
    Mesh Scale Mode=::NewEnumerator3
    Position Offset=<0 0 -60>
    Sprite Size=<20 50>
    Sprite Size Mode=::NewEnumerator4
    UsePositionOffset=true
    Write Material Random=true
   )
  ]
  particleUpdateScript [
   &ParticleState(Kill Particles When Lifetime Has Elapsed=false)
   &ScaleColor(
    Scale Alpha=&RandomRangeFloat(Minimum=0.5, Maximum=1, Evaluation Type=::NewEnumerator1)
    ScaleA=true
    ScaleRGB=true
   ) {enabled=false, module=ScaleColor001}
   /SetVariables(
    Particles.EyeGlowColor:LinearColor=&MultiplyLinearColor(
     Linear Color=<rgba 0.9682 0.984 1 1>
     Scale Factor=&MakeVector4(X=0.4, Y=0.4, Z=0.4, W=$User.GlobalGlow)
    )
    Particles.GlowCameraOffset=3
    Particles.SpikeGlowColor:LinearColor=&MultiplyLinearColor(
     Linear Color=&SelectLinearColorFromArray(
      Array Sampling Mode=::NewEnumerator1
      Direct Array Index=$User.VariantColor
      Color Selection Array=DataInterface() {propertyValues {ColorData [<rgba 1 0.4735 0.107 1>, <rgba 0.0429 0.5848 1 1>, <rgba 1 0.8434 0.1534 1>]}}
      Random Seed=?
     )
     Scale Factor=&MakeVector4(X=0.4, Y=0.4, Z=0.4, W=$User.GlobalGlow)
    )
   ) {isSetParameters=true}
   &SkeletalMeshLocation(Bone Sampling Mode=::NewEnumerator4, MaskSpawnGroup=false, Sampled Position Offset=<0 1 0>, Skeletal Mesh=$System.Skeletal Mesh)
   &ScaleColor(
    Color Value To Scale=$Particles.GlowColor
    Scale Alpha=&FloatFromCurve(FloatCurve=DataInterface() {propertyValues {Curve {keys [<at 0 0>,<at 0.1 1>,<at 1 1>]}}})
    ScaleA=true
    ScaleRGB=true
   ) {enabled=false}
  ]
  rendererClasses [&NiagaraMeshRendererProperties]
  renderers [
   {rendererClass=&NiagaraMeshRendererProperties, rendererIndex=0}
   {rendererClass=&NiagaraMeshRendererProperties, rendererIndex=1}
  ]
 }
}
summary {
 systemName=NS_Companion_Emote_WindTyro
 emitters [
  {emitterName=Lightning, rendererClasses [&NiagaraMeshRendererProperties, &NiagaraLightRendererProperties]}
  {emitterName=Sparks001, rendererClasses [&NiagaraSpriteRendererProperties]}
  {emitterName=Smoke, rendererClasses [&NiagaraSpriteRendererProperties]}
  {emitterName=YellowLightMesh, rendererClasses [&NiagaraMeshRendererProperties]}
 ]
 userVariables [
  User.GlobalGlow {type=NiagaraFloat, defaultValue=1, description=""}
  User.LightningBreath {type=NiagaraFloat, defaultValue=1, description=""}
  User.Smoke {type=NiagaraFloat, defaultValue=0, description=""}
  User.SpawnTime {type=NiagaraFloat, defaultValue=0, description=""}
  User.Tertiary Color {type=LinearColor, description="", defaultValue {struct=LinearColor, value=<rgba 1 1 1 1>}}
  User.VariantColor {type=NiagaraInt32}
  User.VelocityScale {type=NiagaraFloat, defaultValue=0, description=""}
 ]
}
