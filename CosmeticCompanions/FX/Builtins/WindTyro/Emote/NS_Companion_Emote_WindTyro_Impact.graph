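# One-shot ground impact effect for the WindTyro companion emote, with four burst emitters spawning around User.Offset.
# GroundDust and GPU SmallBits sprites compute curve-driven LerpPosition offsets from their spawn shape instead of using velocity.
# GroundDust also animates three sets of dynamic material parameters, while short-lived Sand sprites fly outward and stretch.
# RocksBurst divides tumbling debris across three mesh renderers, adding system momentum, gravity, drag and collision.
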NiagaraSystem /CosmeticCompanions/FX/Builtins/WindTyro/Emote/NS_Companion_Emote_WindTyro_Impact
dependencies {
 usedDataInterfaces [
  &NiagaraDataInterfaceCurve
  &NiagaraDataInterfaceMeshRendererInfo
  &NiagaraDataInterfaceVector2DCurve
 ]
 usedDynamicInputs []
 usedModules [&SystemState]
}
emitters {
 GroundDust {
  emitterUpdateScript [
   &EmitterState()
   &SpawnBurst_Instantaneous(Spawn Count=14, Use Spawn Probability=false, UseLoopCountLimit=false)
   /SetVariables() {isSetParameters=true}
  ]
  particleSpawnScript [
   /SetVariables(Particles.DustColor:LinearColor=&RandomRangeLinearColor(Minimum=<rgba 0.1146 0.1041 0.0995 0.6>, Maximum=<rgba 0.1354 0.1249 0.1165 1>)) {isSetParameters=true}
   /SetVariables(Particles.SizeMax=43.75, Particles.SizeMin=25) {isSetParameters=true}
   /SetVariables(Particles.SizeCurve:NiagaraFloat=&Lerp_Float(
     A=$Particles.SizeMin
     B=$Particles.SizeMax
     Alpha=&FloatFromCurve(
      FloatCurve=DataInterface() {propertyValues {Curve {keys [
          {interpMode=RCIM_Cubic, tangentMode=RCTM_Break, time=0, value=0}
          {arriveTangentWeight=0.9217, interpMode=RCIM_Cubic, tangentMode=RCTM_Break, time=1, value=1}
         ]}}}
      CurveIndex=&RandomRangeFloat(Minimum=0, Maximum=1)
     )
    )) {isSetParameters=true}
   /SetVariables(Particles.InitSpriteRotRate=&UniformRangedFloat(Minimum=-1, Maximum=1)) {isSetParameters=true}
   /SetVariables(Particles.CylinderRadius:NiagaraFloat=&RandomRangeFloat(Minimum=5, Maximum=15)) {isSetParameters=true}
   &InitializeParticle(
    Color=$Particles.DustColor
    Lifetime Max=1.5
    Lifetime Min=0.9
    Lifetime Mode=::NewEnumerator1
    Sprite Rotation Mode=::NewEnumerator2
    Sprite Size Mode=::NewEnumerator3
    Uniform Sprite Size=$Particles.SizeCurve
    UsePositionOffset=false
    Write Material Random=true
   )
   &ShapeLocation(
    Apply To Particle Position=::NewEnumerator1
    Cylinder Distribution=::NewEnumerator2
    Cylinder Height=&RandomRangeFloat(Minimum=0, Maximum=10)
    Cylinder Height Midpoint=0
    Cylinder Radius=$Particles.CylinderRadius
    MaskSpawnGroup=false
    Offset:NiagaraPosition=$User.Offset
    Offset Mode=::NewEnumerator0
    Shape Primitive=::NewEnumerator1
    U Position=&ReturnNormalizedExecIndex()
    UseOwnerScale=true
   )
   &CameraOffset(Camera Offset Amount=&RandomRangeFloat(Minimum=35, Maximum=50))
  ]
  particleUpdateScript [
   &ParticleState()
   &ScaleSpriteSize(
    Scale Sprite Size Mode=::NewEnumerator0
    Uniform Scale Factor=&FloatFromCurve(FloatCurve=DataInterface() {propertyValues {Curve {keys [
         {interpMode=RCIM_Cubic, leaveTangent=3.9827, tangentMode=RCTM_Break, time=0, value=0.2}
         <at 0.1 0.3871>
         {arriveTangent=0.5926, interpMode=RCIM_Cubic, leaveTangent=0.5182, tangentMode=RCTM_Break, time=0.5, value=0.55}
        ]}}})
   )
   &DynamicMaterialParameters(
    Index 0 Param 1=&Add_Float(
     A=&RandomRangeFloat(Minimum=0, Maximum=1)
     B=&FloatFromCurve(
      FloatCurve=DataInterface() {propertyValues {Curve {keys [{tangentMode=RCTM_User, time=0, value=0}, <at 0.2 0.25>, <at 0.8 0.35>, {tangentMode=RCTM_User, time=1, value=1}]}}}
      CurveIndex=&Add_Float(A=&Multiply_Float(A=$Particles.Initial.MaterialRandom, B=1), B=$Particles.NormalizedAge)
      Scale Curve=0.1
     )
    )
    Index 0 Param 2=&FloatFromCurve(FloatCurve=DataInterface() {propertyValues {Curve {keys [<at 0 1>,<at 1 0>]}}})
    Index 0 Param 3=&FloatFromCurve(
     FloatCurve=DataInterface() {propertyValues {Curve {keys [
         {interpMode=RCIM_Cubic, leaveTangent=16.5652, leaveTangentWeight=0.4385, tangentMode=RCTM_Break, time=0, value=0}
         {arriveTangentWeight=0.0333, interpMode=RCIM_Cubic, leaveTangentWeight=0.9477, tangentMode=RCTM_Break, time=0.1, value=1}
         {arriveTangent=-1.7159, interpMode=RCIM_Cubic, leaveTangent=-1.7159, tangentMode=RCTM_User, time=0.3, value=0.2}
         {arriveTangent=0.2347, interpMode=RCIM_Cubic, leaveTangent=0.2723, tangentMode=RCTM_Break, time=0.6, value=0}
        ]}}}
     Scale Curve=0
    )
    Index 0 Param 4=&FloatFromCurve(FloatCurve=DataInterface() {propertyValues {Curve {keys [{interpMode=RCIM_Cubic, tangentMode=RCTM_Break, time=0.5, value=1}, {interpMode=RCIM_Cubic, tangentMode=RCTM_Break, time=1, value=0}]}}})
    Index 1 Param 1=&FloatFromCurve(FloatCurve=DataInterface() {propertyValues {Curve {keys [
         {interpMode=RCIM_Cubic, leaveTangentWeight=0.0833, tangentMode=RCTM_Break, time=0, value=0}
         <at 0.1 0.05>
         <at 0.6 0.1>
         {arriveTangentWeight=0.0873, interpMode=RCIM_Cubic, tangentMode=RCTM_Break, time=0.8, value=1}
        ]}}})
    Index 1 Param 2=&FloatFromCurve(
     FloatCurve=DataInterface() {propertyValues {Curve {keys [
         {interpMode=RCIM_Cubic, tangentMode=RCTM_Break, time=0, value=0.2}
         <at 0.2 0.5>
         <at 0.7 0.575>
         {interpMode=RCIM_Cubic, leaveTangent=0.9066, tangentMode=RCTM_Break, time=0.9, value=2}
        ]}}}
     Scale Curve=&RandomRangeFloat(Minimum=0.1, Maximum=0.2)
    )
    Index 1 Param 3=&FloatFromCurve(FloatCurve=DataInterface() {propertyValues {Curve {keys [
         {interpMode=RCIM_Cubic, leaveTangent=4.4191, tangentMode=RCTM_Break, time=0, value=0}
         <at 0.1 0.225>
         <at 0.5 0.25>
         {arriveTangent=1.7549, interpMode=RCIM_Cubic, leaveTangent=10.6767, tangentMode=RCTM_User, time=0.6, value=0.875}
        ]}}})
    Index 1 Param 4=&FloatFromCurve(
     FloatCurve=DataInterface() {propertyValues {Curve {keys [
         {arriveTangent=0.1828, interpMode=RCIM_Cubic, leaveTangent=0.1828, tangentMode=RCTM_User, time=0, value=0.9575}
         <at 0.2 0.9775>
         <at 0.8 0.975>
         <at 1 0.95>
        ]}}}
     Scale Curve=2
    )
    Index 2 Param 1=&FloatFromCurve(
     FloatCurve=DataInterface() {propertyValues {Curve {keys [
         {interpMode=RCIM_Cubic, leaveTangent=4.2485, tangentMode=RCTM_Break, time=0, value=0}
         <at 0.1 0.3>
         {arriveTangent=1.0548, interpMode=RCIM_Cubic, leaveTangent=1.0548, tangentMode=RCTM_User, time=0.5, value=0.5}
         <at 1 1>
        ]}}}
     Scale Curve=3
    )
    Index 2 Param 2=&FloatFromCurve(
     FloatCurve=DataInterface() {propertyValues {Curve {keys [
         {arriveTangent=1.9385, interpMode=RCIM_Cubic, leaveTangent=1.9385, tangentMode=RCTM_User, time=0, value=0}
         <at 0.2 0.35>
         {arriveTangent=0.8125, interpMode=RCIM_Cubic, leaveTangent=0.8125, time=0.5, value=0.45}
         {interpMode=RCIM_Cubic, tangentMode=RCTM_User, time=1, value=1}
        ]}}}
     Scale Curve=0.3
    )
    Index 2 Param 3=&FloatFromCurve(
     FloatCurve=DataInterface() {propertyValues {Curve {keys [
         {interpMode=RCIM_Cubic, leaveTangent=-4.9239, leaveTangentWeight=0.1411, tangentMode=RCTM_Break, time=0, value=0.7}
         {arriveTangent=-0.3602, interpMode=RCIM_Cubic, leaveTangent=-0.2832, tangentMode=RCTM_Break, time=0.1, value=0.15}
         {arriveTangent=-0.2135, interpMode=RCIM_Cubic, leaveTangent=-0.2135, tangentMode=RCTM_User, time=0.7, value=0.05}
         {arriveTangent=0.8818, arriveTangentWeight=1.057, interpMode=RCIM_Cubic, tangentMode=RCTM_Break, time=1, value=1}
        ]}}}
     Scale Curve=5
    )
    Index 2 Param 4=&RandomRangeFloat(Minimum=1.5, Maximum=1.75)
    Module.Param0WriteEnabled=true
    Module.Param0WriteEnabled1=true
    Module.Param0WriteEnabled2=true
    Module.Param1WriteEnabled=false
    Module.Param1WriteEnabled1=true
    Module.Param1WriteEnabled2=true
    Module.Param2WriteEnabled=true
    Module.Param2WriteEnabled1=true
    Module.Param2WriteEnabled2=true
    Module.Param3WriteEnabled=true
    Module.Param3WriteEnabled1=true
    Module.Param3WriteEnabled2=true
    Write Parameter Index 1=true
    Write Parameter Index 2=true
   )
   &ScaleColor(
    Scale Alpha=&FloatFromCurve(
     FloatCurve=DataInterface() {propertyValues {Curve {keys [
         {arriveTangent=5.1462, interpMode=RCIM_Cubic, leaveTangent=5.1462, tangentMode=RCTM_User, time=0, value=0}
         {interpMode=RCIM_Cubic, tangentMode=RCTM_User, time=0.1, value=1}
         <at 0.9854 0 RCIM_Cubic>
        ]}}}
     Scale Curve=&RandomRangeFloat(Minimum=0.75, Maximum=1)
    )
    ScaleA=true
    ScaleRGB=false
   )
   /SetVariables(Particles.LerpPosition:NiagaraPosition=&Add_Vector(
     A=$Particles.ShapeLocation.ShapePosition
     B=&Multiply_Vector(
      A=$Particles.ShapeLocation.ShapeNormal
      B=&Multiply_VectorByFloat(
       Vector=&RandomRangeVector(Minimum=<130 130 0>, Maximum=<180 180 0>)
       Float=&FloatFromCurve(
        FloatCurve=DataInterface() {propertyValues {Curve {keys [
            {arriveTangent=1.4325, interpMode=RCIM_Cubic, leaveTangent=1.4325, tangentMode=RCTM_User, time=0, value=0}
            {arriveTangent=0.1676, interpMode=RCIM_Cubic, leaveTangent=0.1676, tangentMode=RCTM_User, time=1, value=0.4}
           ]}}}
        Scale Curve=0.2
       )
      )
     )
    )) {isSetParameters=true}
  ]
  rendererClasses [&NiagaraSpriteRendererProperties]
  renderers [
   {rendererClass=&NiagaraSpriteRendererProperties}
   {rendererClass=&NiagaraSpriteRendererProperties, rendererIndex=1}
  ]
 }
 RocksBurst {
  emitterUpdateScript [
   &EmitterState()
   &SpawnBurst_Instantaneous(
    Spawn Count=&RandomRangeInt(Minimum=3, Maximum=4)
    Use Spawn Probability=false
    UseLoopCountLimit=false
   )
  ]
  particleSpawnScript [
   &InitializeParticle(
    Lifetime Max=1.5
    Lifetime Min=0.5
    Lifetime Mode=::NewEnumerator1
    Mesh Scale Max=<0.5 0.5 0.5>
    Mesh Scale Min=<0.2 0.2 0.2>
    Mesh Scale Mode=::NewEnumerator2
    UsePositionOffset=false
    Write Material Random=true
   )
   &ShapeLocation(
    Handle Radius=15
    Large Radius=5
    MaskSpawnGroup=false
    Offset:NiagaraPosition=$User.Offset
    Offset Mode=::NewEnumerator0
    Override Seed=false
    Shape Primitive=::NewEnumerator3
    UseOwnerScale=true
   )
   &InitialMeshOrientation(Use Rotation Vector=false)
   &self:DivideParticles(HowManyRenderers=3) {module=DivideParticles}
   &AddRotationalVelocity(Rotation Rate=&Multiply_VectorByFloat(
     Vector=&RandomVector(Vector Scale=2)
     Float=&RandomRangeFloat(Minimum=0.5, Maximum=1)
    ))
   &DynamicMaterialParameters(
    Index 0 Param 2=&RandomRangeFloat(Minimum=0.25, Maximum=1)
    Module.Param0WriteEnabled=false
    Module.Param1WriteEnabled=true
    Module.Param2WriteEnabled=false
    Module.Param3WriteEnabled=false
   )
   &AddVelocity(
    Module.Use Falloff Distance=false
    Origin Offset:NiagaraPosition=$User.Offset
    Override Seed=false
    Velocity Mode=::NewEnumerator1
    Velocity Speed=&RandomRangeFloat(Minimum=30, Maximum=140)
   ) {module=AddVelocity001}
   &AddVelocity(Velocity=&RandomRangeVector(Minimum=<0 0 100>, Maximum=<0 0 160>), Velocity Speed Scale=2)
   &AddVelocity(Velocity=$System.Momentum) {module=AddVelocity002}
   &ApplyInitialForces()
  ]
  particleUpdateScript [
   &ParticleState()
   &GravityForce(Gravity=&RandomRangeVector(Minimum=<0 0 -750>, Maximum=<0 0 -1500>))
   &Collision(
    Age Colliding Particles=false
    Module.Control Rest State Via Surface Normal=false
    Module.EnableMaxCollisionCount=false
    Module.Randomize Collision Normal Vector=false
    Restitution=0.1
   )
   &Drag(Drag=1.6, Use Linear Drag=true, Use Rotational Drag=false)
   &SolveForcesAndVelocity(Clamp Velocity=false, Limit Acceleration=false)
   &ScaleMeshSize(Scale Factor=&VectorFromFloat(Value=&FloatFromCurve(
      FloatCurve=DataInterface() {propertyValues {Curve {keys [{arriveTangent=-0.0035, interpMode=RCIM_Cubic, leaveTangent=-0.0035, tangentMode=RCTM_User, time=0.4, value=1}, <at 1 0>]}}}
      Scale Curve=&RandomRangeFloat(Minimum=0.2, Maximum=0.3)
     )))
   &SolveRotationalForcesAndVelocity()
  ]
  rendererClasses [&NiagaraMeshRendererProperties]
  renderers [
   {rendererClass=&NiagaraMeshRendererProperties, rendererIndex=0}
   {rendererClass=&NiagaraMeshRendererProperties, rendererIndex=1}
   {rendererClass=&NiagaraMeshRendererProperties, rendererIndex=2}
  ]
 }
 Sand {
  emitterUpdateScript [
   &EmitterState()
   &SpawnBurst_Instantaneous(Spawn Count=10, Use Spawn Probability=false, UseLoopCountLimit=false)
   /SetVariables(Emitter.DepthFade=40) {isSetParameters=true}
  ]
  particleSpawnScript [
   &InitializeParticle(
    Color Maximum=<rgba 0.2708 0.2506 0.2327 0.5>
    Color Minimum=<rgba 0.0729 0.0681 0.063 0.5>
    Color Mode=::NewEnumerator2
    Lifetime Max=0.3
    Lifetime Min=0.2
    Lifetime Mode=::NewEnumerator1
    Position Offset=<0 0 -10>
    Sprite Rotation Mode=::NewEnumerator2
    Sprite Size Max=<35 60>
    Sprite Size Min=<25 50>
    Sprite Size Mode=::NewEnumerator2
    UsePositionOffset=true
    Write Material Random=true
   )
   &ShapeLocation(
    Handle Radius=2
    Large Radius=10
    MaskSpawnGroup=false
    Offset:NiagaraPosition=$User.Offset
    Offset Mode=::NewEnumerator0
    Override Seed=false
    Shape Origin=$Particles.Position
    Shape Primitive=::NewEnumerator3
    UseOwnerScale=true
   )
   &SubUVAnimation(End Frame=3, SubUV Animation Mode:ENiagara_SubUVLookupMode=::NewEnumerator2)
   &AddVelocity(
    Module.Use Falloff Distance=false
    Origin Offset=&AddVectorToPosition(Position=$User.Offset, Vector=<0 0 -35>)
    Origin Offset Coordinate Space=::Simulation
    Override Seed=false
    Velocity Mode=::NewEnumerator1
    Velocity Speed=150
   )
   &ApplyInitialForces()
  ]
  particleUpdateScript [
   &ParticleState()
   &SolveForcesAndVelocity(Clamp Velocity=false, Limit Acceleration=false)
   &ScaleSpriteSize(Scale Factor=&Vector2DFromCurve(
     Vector2Curve=DataInterface() {propertyValues {XCurve {keys [<at 0.0036 1.6897>,<at 0.9967 0.0582>]}, YCurve {keys [<at 0 1>,<at 1.002 3.7227>]}}}
     ScaleCurve=<0.3 0.3>
    ))
   &ScaleColor(
    Scale Alpha=&FloatFromCurve(
     FloatCurve=DataInterface() {propertyValues {Curve {keys [
         {interpMode=RCIM_Cubic, leaveTangentWeight=0.0962, tangentMode=RCTM_Break, time=0, value=0}
         {interpMode=RCIM_Cubic, tangentMode=RCTM_Break, time=0.2, value=1}
         {interpMode=RCIM_Cubic, tangentMode=RCTM_Break, time=1, value=0}
        ]}}}
     Scale Curve=&RandomRangeFloat(Minimum=0.35, Maximum=0.8)
    )
    ScaleA=true
    ScaleRGB=false
   )
   &DynamicMaterialParameters(
    Index 0 Param 2=1
    Index 0 Param 3=1
    Index 0 Param 4=&FloatFromCurve(FloatCurve=DataInterface() {propertyValues {Curve {keys [{interpMode=RCIM_Cubic, tangentMode=RCTM_Break, time=0.3, value=1}, {interpMode=RCIM_Cubic, tangentMode=RCTM_Break, time=1, value=0.25}]}}})
    Module.Param0WriteEnabled=true
    Module.Param1WriteEnabled=true
    Module.Param2WriteEnabled=true
    Module.Param3WriteEnabled=true
   )
  ]
  rendererClasses [&NiagaraSpriteRendererProperties]
  renderers [{rendererClass=&NiagaraSpriteRendererProperties}]
 }
 SmallBits {
  simTarget=GPUComputeSim
  emitterUpdateScript [
   &EmitterState()
   &SpawnBurst_Instantaneous(
    Spawn Count=&RandomRangeInt(Minimum=20, Maximum=30)
    Use Spawn Probability=false
    UseLoopCountLimit=false
   )
  ]
  particleSpawnScript [
   /SetVariables(Particles.randomInt:NiagaraInt32=&RandomRangeInt(Minimum=0, Maximum=1)) {isSetParameters=true}
   &InitializeParticle(
    Color Maximum=<rgba 0.4323 0.4287 0.3936 1>
    Color Minimum=<rgba 0.3108 0.3125 0.2834 1>
    Color Mode=::NewEnumerator2
    Lifetime=&RandomRangeFloat(Minimum=1, Maximum=1.4)
    Sprite Rotation Mode=::NewEnumerator2
    Sprite Size Mode=::NewEnumerator1
    Uniform Sprite Size Max=2
    Uniform Sprite Size Min=1
    UsePositionOffset=false
    Write Material Random=true
   )
   &SubUVAnimation(End Frame=3, SubUV Animation Mode:ENiagara_SubUVLookupMode=::NewEnumerator2)
   &ShapeLocation(
    Cylinder Height=0
    Cylinder Radius=&RandomRangeFloat(Minimum=10, Maximum=3)
    MaskSpawnGroup=false
    Offset:NiagaraPosition=$User.Offset
    Offset Mode=::NewEnumerator0
    Override Seed=false
    Shape Primitive=::NewEnumerator1
    Surface Only=false
    UseOwnerScale=true
   ) {module=ShapeLocation001}
   /SetVariables(Particles.InitSpriteRotRate=&UniformRangedFloat(Minimum=-3, Maximum=3)) {isSetParameters=true}
  ]
  particleUpdateScript [
   &ParticleState()
   /SetVariables(Particles.LerpPosition:NiagaraPosition=&Add_Vector(
     A=$Particles.ShapeLocation001.ShapePosition
     B=&MakeVector(
      X=&MakeFloatFromVector(
       Vector=&Multiply_VectorByFloat(
        Vector=$Particles.ShapeLocation001.ShapeNormal
        Float=&FloatFromCurve(
         FloatCurve=DataInterface() {propertyValues {Curve {keys [{tangentMode=RCTM_User, time=0, value=0}, {tangentMode=RCTM_User, time=0.6, value=1}]}}}
         Scale Curve=&RandomRangeFloat(Minimum=43.75, Maximum=56.25)
        )
       )
       Channel=::NewEnumerator0
      )
      Y=&MakeFloatFromVector(
       Vector=&Multiply_VectorByFloat(
        Vector=$Particles.ShapeLocation001.ShapeNormal
        Float=&FloatFromCurve(
         FloatCurve=DataInterface() {propertyValues {Curve {keys [{tangentMode=RCTM_User, time=0, value=0}, {tangentMode=RCTM_User, time=0.6, value=1}]}}}
         Scale Curve=&RandomRangeFloat(Minimum=43.75, Maximum=56.25)
        )
       )
       Channel=::NewEnumerator1
      )
      Z=&MakeFloatFromVector(
       Vector=&Multiply_VectorByFloat(
        Vector=$Particles.ShapeLocation001.ShapeTangent
        Float=&FloatFromCurve(
         FloatCurve=DataInterface() {propertyValues {Curve {keys [
             {
              arriveTangent=35.9693
              arriveTangentWeight=2.3198
              interpMode=RCIM_Cubic
              leaveTangent=35.9693
              leaveTangentWeight=2.3198
              tangentMode=RCTM_User
              tangentWeightMode=RCTWM_WeightedBoth
              time=0
              value=0
             }
             {
              arriveTangent=-12.5267
              arriveTangentWeight=2.3337
              interpMode=RCIM_Cubic
              leaveTangent=-12.5267
              leaveTangentWeight=2.3337
              tangentMode=RCTM_User
              tangentWeightMode=RCTWM_WeightedBoth
              time=0.5
              value=-0.375
             }
            ]}}}
         Scale Curve=&FloatFromCurve(
          FloatCurve=DataInterface() {propertyValues {Curve {keys [{tangentMode=RCTM_User, time=0, value=1}, {interpMode=RCIM_Cubic, leaveTangent=-2.375, time=0.6, value=1.5}]}}}
          Scale Curve=&RandomRangeFloat(Minimum=6.6667, Maximum=35)
         )
        )
       )
       Channel=::NewEnumerator2
      )
     )
    )) {isSetParameters=true}
   &ScaleSpriteSize(
    Uniform Curve Scale=2.5
    Uniform Curve Sprite Scale=DataInterface() {propertyValues {Curve {keys [
        {arriveTangent=40, interpMode=RCIM_Cubic, time=0, value=0}
        {arriveTangent=-1.7622, interpMode=RCIM_Cubic, leaveTangent=-1.7622, tangentMode=RCTM_User, time=0.025, value=1}
        {arriveTangent=-0.5297, interpMode=RCIM_Cubic, leaveTangent=-0.5297, tangentMode=RCTM_User, time=1, value=0}
       ]}}}
   )
  ]
  rendererClasses [&NiagaraSpriteRendererProperties]
  renderers [{rendererClass=&NiagaraSpriteRendererProperties}]
 }
}
summary {
 systemName=NS_Companion_Emote_WindTyro_Impact
 emitters [
  {emitterName=GroundDust, rendererClasses [&NiagaraSpriteRendererProperties]}
  {emitterName=Sand, rendererClasses [&NiagaraSpriteRendererProperties]}
  {emitterName=SmallBits, simTarget=GPUComputeSim, rendererClasses [&NiagaraSpriteRendererProperties]}
  {emitterName=RocksBurst, rendererClasses [&NiagaraMeshRendererProperties]}
 ]
 userVariables [
  User.DesertBoulder {type=NiagaraBool}
  User.Offset {type=NiagaraPosition, description="", defaultValue {}}
 ]
}
